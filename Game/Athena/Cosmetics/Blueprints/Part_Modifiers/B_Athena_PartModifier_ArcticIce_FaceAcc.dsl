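; Cosmetic part modifier for ArcticIce face accessories that shows the scarf only while a supported backpack is visible.
; Part change, visibility and cosmetic-applied delegates track the backpack against a supported-parts list and recompute scarf state.
; State changes and post res-in push material parameters to Head and Face meshes, rebinding for newly applied parts.
; Talus-specific handling swaps head overlay outlines for caped variants, toggles face outlines and hides elements in front-end holo mode.
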(event EventOnPawnAndPartChanged
  (bind _returnvalue (CustomCharacter|GetAssociatedPlayerPawn))
  (|CallParentFunction)
  (Utilities|IsValid _returnvalue
    (:Is%20Valid
      (Utilities|FlowControl|Sequence
        (:then_0
          (EventDispatchers|BindEventtoOnPartChanged _returnvalue (EventDispatchers|CreateEvent :!event Handle%20Part%20Changed))
          (EventDispatchers|BindEventtoOnManagedPartVisibilityChanged _returnvalue (EventDispatchers|CreateEvent :!event Handle%20Part%20Visibility%20Changed)))
        (:then_1
          (Variables|Default|SetBackpack%20Visible (CharacterPart|IsPartVisibleAccordingtoCount _returnvalue "Backpack"))
          (bind _delegatehandlecontroller_out (Cosmetics|WhenCosmeticSpecificPartApplied _returnvalue (EventDispatchers|CreateEvent :!event Check%20Backpack%20Part) false "Backpack"))
          (Variables|Default|SetBackpack%20Delegate%20Controller _delegatehandlecontroller_out))
        (:then_2
          (CallFunction|CheckScarfVisible))))))

(event EventOnPostResIn
  (|CallParentFunction)
  (Default|UpdateMaterialParameters))

(event EventBeginPlay
  (|CallParentFunction))

(fn CheckBackpackPart
  (params (Pawn :type Fort%20Player%20Pawn%20Object%20Reference) (Part :type Custom%20Character%20Part%20Object%20Reference) (MeshPart :type Skeletal%20Mesh%20Component%20Object%20Reference) (PartType :type EFortCustomPartType%20Enum :default "Head"))
  (Utilities|IsValid Pawn
    (:Is%20Valid
      (bind _supported (CallFunction|IsBackpackSupported Part))
      (bind _output_get (LocalVariables|SetNew%20Supported%20Backpack _supported))
      (if (!= _output_get (Variables|Default|GetSupported%20Backpack))
        (Variables|Default|SetSupported%20Backpack (LocalVariables|GetNew%20Supported%20Backpack))
        (CallFunction|CheckScarfVisible)))))

(fn CheckScarfVisible
  (bind _output_get (LocalVariables|SetNew%20Scarf%20Visible (and (Variables|Default|GetSupported%20Backpack) (Variables|Default|GetBackpack%20Visible))))
  (if (!= _output_get (Variables|Default|GetScarf%20Visible))
    (Variables|Default|SetScarf%20Visible (LocalVariables|GetNew%20Scarf%20Visible))
    (Default|UpdateMaterialParameters)))

(fn HandlePartChanged
  (params (PartType :type EFortCustomPartType%20Enum :default "Head"))
  (bind _returnvalue (CustomCharacter|GetAssociatedPlayerPawn))
  (if (Utilities|Enum|Equal(Enum) PartType "Backpack")
    (Utilities|IsValid _returnvalue
      (:Is%20Valid
        (Cosmetics|WhenCosmeticSpecificPartApplied _returnvalue (EventDispatchers|CreateEvent :!event Check%20Backpack%20Part) false "Backpack")))))

(fn HandlePartVisibilityChanged
  (params (PartType :type EFortCustomPartType%20Enum :default "Head") (bVisible :type Boolean))
  (if (Utilities|Enum|Equal(Enum) PartType "Backpack")
    (if (!= (Variables|Default|GetBackpack%20Visible) (Variables|GetbVisible))
      (Variables|Default|SetBackpack%20Visible (Variables|GetbVisible))
      (CallFunction|CheckScarfVisible))))

(fn IsBackpackSupported
  (params (Character%20Part :type Custom%20Character%20Part%20Object%20Reference))
  (for _array_element (Variables|Default|GetSupported%20Character%20Parts)
    (if (== (Utilities|ResolveSoftReference _array_element) (Variables|GetCharacter%20Part))
      (return true)))
  (return false))

(fn PushFaceMaterialParameters
  (params (MeshPart :type Skeletal%20Mesh%20Component%20Object%20Reference))
  (bind _scarf_visible (Variables|Default|GetScarf%20Visible))
  (bind _returnvalue (not _scarf_visible))
  (bind _returnvalue_2 (Math|Conversions|ToFloat(Boolean) _returnvalue))
  (bind _returnvalue_3 (Rendering|Material|GetOverlayMaterial MeshPart false 0 :!class /Script/Engine.MeshComponent))
  (bind _scarf_visible_2 (Variables|Default|GetScarf%20Visible))
  (Utilities|IsValid MeshPart
    (:Is%20Valid
      ;[1 1 1] Hide Scarf in Holo Mode
      (bind _asmaterial_instance_dynamic (Utilities|Casting|CastToMaterialInstanceDynamic (Utilities|Array|Get(acopy) (Rendering|Material|GetMaterials MeshPart) 1))
        (:then
          (Rendering|Material|SetScalarParameterValue _asmaterial_instance_dynamic "HT_CrunchVerts" _returnvalue_2 :!class /Script/Engine.MaterialInstanceDynamic))
        (:CastFailed
          _))
      ;[1 1 1] Hide ArcticIceTalus head wrap in FE holo mode
      (if (> (Utilities|String|FindSubstring (Utilities|GetDisplayName MeshPart) "Talus") -1)
        (Rendering|Material|SetScalarParameterValue (Class|BTechArtGlobalLibrary|GetDynamicMaterialInstance MeshPart) "HT_CrunchVerts" _returnvalue_2 :!class /Script/Engine.MaterialInstanceDynamic))
      (Rendering|Material|SetScalarParameterValueonMaterials MeshPart "ArcticIce_Visibility" (Math|Conversions|ToFloat(Boolean) _scarf_visible))
      (Development|PrintString (Utilities|String|Append "Scarf Visible: " (Utilities|String|ToString(Boolean) _scarf_visible_2)))
      ;[0.1019 0.651 0.07776] Talus FaceAcc: show low quality overlay outline when cape is present, otherwise hide outline
      (if (CustomCosmeticModifier|IsinFrontEnd)
        (bind _asmaterial_instance_dynamic_2 (Utilities|Casting|CastToMaterialInstanceDynamic _returnvalue_3))
        (if (== (Variables|/Script/Engine.MaterialInstance|GetParent _asmaterial_instance_dynamic_2) "/Game/Characters/Player/Male/Medium/Bodies/M_MED_ArcticIceTalus/Materials/MI_MED_ArcticIceTalus_OverlayOutlines_Acc.MI_MED_ArcticIceTalus_OverlayOutlines_Acc")
          _)
        (elif (== _returnvalue_3 "/Game/Characters/Player/Male/Medium/Bodies/M_MED_ArcticIceTalus/Materials/MI_MED_ArcticIceTalus_OverlayOutlines_Acc.MI_MED_ArcticIceTalus_OverlayOutlines_Acc")
          _))
      (Class|FortCosmeticMaterialBlueprintLibrary|SetScalarParameterValueonOverlayMaterial MeshPart "OutlineVisibility" (Math|Conversions|ToFloat(Boolean) _scarf_visible_2)))))

(fn PushHeadMaterialParameters
  (params (MeshPart :type Skeletal%20Mesh%20Component%20Object%20Reference))
  (bind _returnvalue (Rendering|Material|GetOverlayMaterial MeshPart false 0 :!class /Script/Engine.MeshComponent))
  (bind _returnvalue_2 (CustomCosmeticModifier|IsinFrontEnd))
  (bind _scarf_visible (Variables|Default|GetScarf%20Visible))
  (Utilities|IsValid MeshPart
    (:Is%20Valid
      (Rendering|Material|SetScalarParameterValueonMaterials MeshPart "Hide Element 10" (Math|Conversions|ToFloat(Boolean) _scarf_visible))
      ;[0.1019 0.651 0.07776] Talus Head - Handle Overlay Swap to hide low quality overlay outlines
      (if _scarf_visible
        (if _returnvalue_2
          (bind _asmaterial_instance_dynamic (Utilities|Casting|CastToMaterialInstanceDynamic _returnvalue))
          (if (== (Variables|/Script/Engine.MaterialInstance|GetParent _asmaterial_instance_dynamic) "/Game/Characters/Player/Male/Medium/Bodies/M_MED_ArcticIceTalus/Materials/MI_MED_ArcticIceTalus_OverlayOutlines_Head.MI_MED_ArcticIceTalus_OverlayOutlines_Head")
            _)
          (elif (== _returnvalue "/Game/Characters/Player/Male/Medium/Bodies/M_MED_ArcticIceTalus/Materials/MI_MED_ArcticIceTalus_OverlayOutlines_Head.MI_MED_ArcticIceTalus_OverlayOutlines_Head")
            _))
        (Rendering|Material|SetOverlayMaterial MeshPart "/Game/Characters/Player/Male/Medium/Bodies/M_MED_ArcticIceTalus/Materials/MI_MED_ArcticIceTalus_OverlayOutlines_Head_Caped.MI_MED_ArcticIceTalus_OverlayOutlines_Head_Caped" false 0 :!class /Script/Engine.MeshComponent)
        (elif _returnvalue_2
          (bind _asmaterial_instance_dynamic_2 (Utilities|Casting|CastToMaterialInstanceDynamic _returnvalue))
          (if (== (Variables|/Script/Engine.MaterialInstance|GetParent _asmaterial_instance_dynamic_2) "/Game/Characters/Player/Male/Medium/Bodies/M_MED_ArcticIceTalus/Materials/MI_MED_ArcticIceTalus_OverlayOutlines_Head_Caped.MI_MED_ArcticIceTalus_OverlayOutlines_Head_Caped")
            (Rendering|Material|SetOverlayMaterial MeshPart "/Game/Characters/Player/Male/Medium/Bodies/M_MED_ArcticIceTalus/Materials/MI_MED_ArcticIceTalus_OverlayOutlines_Head.MI_MED_ArcticIceTalus_OverlayOutlines_Head" false 0 :!class /Script/Engine.MeshComponent))
          (elif (== _returnvalue "/Game/Characters/Player/Male/Medium/Bodies/M_MED_ArcticIceTalus/Materials/MI_MED_ArcticIceTalus_OverlayOutlines_Head_Caped.MI_MED_ArcticIceTalus_OverlayOutlines_Head_Caped")
            _))))))

(fn PushMaterialParameters
  (params (Pawn :type Fort%20Player%20Pawn%20Object%20Reference) (Part :type Custom%20Character%20Part%20Object%20Reference) (MeshPart :type Skeletal%20Mesh%20Component%20Object%20Reference) (PartType :type EFortCustomPartType%20Enum :default "Head"))
  (Utilities|IsValid MeshPart
    (:Is%20Valid
      (switch Utilities|FlowControl|Switch|SwitchonEFortCustomPartType PartType
        (:Head
          (CallFunction|PushHeadMaterialParameters (Variables|GetMeshPart)))
        (:Face
          (CallFunction|PushFaceMaterialParameters (Variables|GetMeshPart)))))))

(fn UpdateMaterialParameters
  (bind _returnvalue (CustomCharacter|GetAssociatedPlayerPawn))
  (Utilities|IsValid _returnvalue
    (:Is%20Valid
      ;[1 1 1] FORT-992376 push material parameters before binding
      (CallFunction|PushHeadMaterialParameters (Customization|GetSkeletalMeshforPartType _returnvalue "Head" :!class /Script/FortniteGame.FortPlayerPawn))
      (CallFunction|PushFaceMaterialParameters (Customization|GetSkeletalMeshforPartType _returnvalue "Face" :!class /Script/FortniteGame.FortPlayerPawn))
      (bind _delegatehandlecontroller_out (Cosmetics|WhenCosmeticPartApplied _returnvalue (EventDispatchers|CreateEvent :!event Push%20Material%20Parameters) true))
      (Variables|Default|SetMaterial%20Update%20Delegate%20Controller _delegatehandlecontroller_out)
      ;[1 1 1] FORT-553429: hide ArcticIceTalus antennas in FE holo mode
      (Rendering|Material|SetScalarParameterValue (Class|BTechArtGlobalLibrary|GetDynamicMaterialInstance (CustomCharacter|GetSkeletalMeshForPartOnPlayerPawn (FortPawn|GetModifierforPartType _returnvalue "Head"))) "HideElementsOnlyConsidersRedChannel" 0.0 :!class /Script/Engine.MaterialInstanceDynamic))))

(fn ConstructionScript
  (|CallParentFunction))
