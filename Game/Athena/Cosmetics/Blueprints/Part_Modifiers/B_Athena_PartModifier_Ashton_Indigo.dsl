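; Character part modifier for the Ashton Indigo cosmetic that drives hand repulsor idle, muzzle and booster effects.
; After spawn, it attaches effect components to left and right repulsor sockets and binds to anim-blueprint boost and dual-weapon fire events, with a delay hack for late-arriving weapons.
; Firing flashes emissive color, briefly hides that hand's idle effect and interrupts boosting.
; Boosting runs a looping timer that polls the IndigoBoost animation curve.
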(event Custom|Open_1
  (Rendering|SetActorHiddenInGame false))

(event EventOnPostResIn
  (|CallParentFunction)
  (CallFunction|MuzzleAttachment)
  (CallFunction|BoostAttachment)
  (CallFunction|BoostNotifySetup)
  (CallFunction|IdleAttachment)
  (CallFunction|MuzzleFiringSetup 0))

(event Custom|Close_1
  (bind _asfortnite_m_avg_player_anim_blueprint (Utilities|Casting|CastToFortnite_M_Avg_Player_AnimBlueprint (Components|SkeletalMesh|GetAnimInstance (Variables|/Script/Engine.Character|GetMesh (CustomCharacter|GetAssociatedPlayerPawn))))
    (:then
      (Default|UnbindEventfromIndigoBoost _asfortnite_m_avg_player_anim_blueprint (EventDispatchers|CreateEvent :!event Custom|BoostActivate)))
    (:CastFailed
      _))
  (bind _asb_ranged_ashton_indigo_athena (Utilities|Casting|CastToB_Ranged_Ashton_Indigo_Athena (Variables|/Script/FortniteGame.FortPawn|GetCurrentWeapon (CustomCharacter|GetAssociatedPlayerPawn)))
    (:then
      (RangedWeaponDual|UnbindEventfromOnHandFired _asb_ranged_ashton_indigo_athena (EventDispatchers|CreateEvent :!event Custom|Firing)))
    (:CastFailed
      _))
  (Rendering|SetActorHiddenInGame true))

(event Custom|Idle%20Attachment
  (bind _returnvalue (CustomCharacter|GetCharacterPartType))
  (bind _returnvalue_2 (CustomCharacter|GetAssociatedPlayerPawn))
  (bind _returnvalue_3 (Customization|GetSkeletalMeshforPartType _returnvalue_2 _returnvalue :!class /Script/FortniteGame.FortPlayerPawn))
  (Transformation|AttachComponentToComponent (Variables|Default|GetIdle_Left) _returnvalue_3 "FX_repulsor_l" "KeepRelative" "KeepRelative" "KeepRelative" true)
  (Transformation|AttachComponentToComponent (Variables|Default|GetIdle_Right) _returnvalue_3 "FX_repulsor_r" "KeepRelative" "KeepRelative" "KeepRelative" true)
  (Effects|Components|ParticleSystem|SetEmitterEnable (Variables|Default|GetIdle_Right) "Booster" false))

(event Custom|Muzzle_Firing_Multicast
  (params (Weapon :type Fort%20Weapon%20Object%20Reference))
  ;[1 0.6369 0.427] Muzzle Firing Setup - Must be Replicated to Function!
  (bind _asb_ranged_ashton_indigo_athena (Utilities|Casting|CastToB_Ranged_Ashton_Indigo_Athena (Variables|/Script/FortniteGame.FortPawn|GetCurrentWeapon (CustomCharacter|GetAssociatedPlayerPawn)))
    (:then
      _)
    (:CastFailed
      (Development|PrintString "CASTFAILED" true true "(R=1.000000,G=0.000000,B=0.031557,A=1.000000)")
      ;[1 0 0] Hack loop to fix current weapon not being valid yet on laggy clients.
      (Utilities|FlowControl|DoN 5
        (:Exit
          (Utilities|FlowControl|Delay 0.1)))))
  (RangedWeaponDual|BindEventtoOnHandFired _asb_ranged_ashton_indigo_athena (EventDispatchers|CreateEvent :!event Custom|Firing))
  (Development|PrintString (Utilities|GetDisplayName _asb_ranged_ashton_indigo_athena) true true "(R=0.290283,G=1.000000,B=0.000000,A=1.000000)"))

(event Custom|Muzzle_Firing_Setup
  (params (WeaponObject :type Fort%20Weapon%20Object%20Reference))
  ;[1 0.6369 0.427] Muzzle Firing Setup - Must be Replicated to Function!
  (CallFunction|MuzzleFiringMulticast 0))

(event Custom|Firing
  (params (Hand :type EDualWeaponHand%20Enum :default "LEFT") (bPersistantFire :type Boolean))
  (bind _returnvalue (CustomCharacter|GetCharacterPartType))
  (bind _returnvalue_2 (CustomCharacter|GetAssociatedPlayerPawn))
  (bind _returnvalue_3 (Customization|GetSkeletalMeshforPartType _returnvalue_2 _returnvalue :!class /Script/FortniteGame.FortPlayerPawn))
  (bind _idle_right (Variables|Default|GetIdle_Right))
  (bind _idle_left (Variables|Default|GetIdle_Left))
  (bind _lmb_color (Variables|Default|GetLMB%20Color))
  (bind _boostcolor (Variables|Default|GetBoostColor))
  ;[1 0.6369 0.427] Firing - IdleControl
  (CallFunction|FireStopsBooster)
  (if (Utilities|Enum|Equal(Enum) Hand "LEFT")
    (Components|Activation|Activate (Variables|Default|GetMuzzle_Left) false)
    (Rendering|Material|SetVectorParameterValueonMaterials _returnvalue_3 "Emissive Color" (Math|Conversions|ToVector(LinearColor) _lmb_color))
    (Development|SetHiddeninGame _idle_left true false)
    (Utilities|FlowControl|Delay 0.3)
    (Development|SetHiddeninGame _idle_left false false)
    (Components|Activation|Activate _idle_left true)
    (Rendering|Material|SetVectorParameterValueonMaterials _returnvalue_3 "Emissive Color" (Math|Conversions|ToVector(LinearColor) _boostcolor))
    (else
      (Components|Activation|Activate (Variables|Default|GetMuzzle_Right) false)
      (Rendering|Material|SetVectorParameterValueonMaterials _returnvalue_3 "Emissive Color" (Math|Conversions|ToVector(LinearColor) _lmb_color))
      (Development|SetHiddeninGame _idle_right true false)
      (Utilities|FlowControl|Delay 0.3)
      (Development|SetHiddeninGame _idle_right false false)
      (Components|Activation|Activate _idle_right true)
      (Rendering|Material|SetVectorParameterValueonMaterials _returnvalue_3 "Emissive Color" (Math|Conversions|ToVector(LinearColor) _boostcolor)))))

(event Custom|Muzzle_Attachment
  (bind _returnvalue (CustomCharacter|GetSkeletalMeshForPartOnPlayerPawn))
  ;[1 0.6369 0.427] Muzzle Setups
  (Transformation|AttachComponentToComponent (Variables|Default|GetMuzzle_Left) _returnvalue "FX_repulsor_l" "KeepRelative" "KeepRelative" "KeepRelative" true)
  (Transformation|AttachComponentToComponent (Variables|Default|GetMuzzle_Right) _returnvalue "FX_repulsor_r" "KeepRelative" "KeepRelative" "KeepRelative" true))

(event Custom|Fire%20Stops%20Booster
  ;[1 1 1] Insure Booster turns off when Firing
  (Utilities|Time|ClearandInvalidateTimerbyHandle (Variables|Default|GetBooster%20Timer))
  (Components|Activation|Deactivate (Variables|Default|GetBoostJet_Idle_Left))
  ;[0.5 1 1] Boost Active - Idle Color Control and Emitter Enabled
  (Utilities|FlowControl|DoOnce false
    (:Completed
      (Components|Activation|Activate (Variables|Default|GetBoostJet_Idle_Right) false))))

(event Custom|Boost_Attachment
  (bind _returnvalue (CustomCharacter|GetSkeletalMeshForPartOnPlayerPawn))
  ;[0.5 1 1] Boost Setups
  (Transformation|AttachComponentToComponent (Variables|Default|GetBoostJet_Left) _returnvalue "FX_repulsor_l" "KeepRelative" "KeepRelative" "KeepRelative" true)
  (Transformation|AttachComponentToComponent (Variables|Default|GetBoostJet_Right) _returnvalue "FX_repulsor_r" "KeepRelative" "KeepRelative" "KeepRelative" true)
  (Transformation|AttachComponentToComponent (Variables|Default|GetBoostJet_Idle_Left) _returnvalue "FX_repulsor_l" "KeepRelative" "KeepRelative" "KeepRelative" true)
  (Transformation|AttachComponentToComponent (Variables|Default|GetBoostJet_Idle_Right) _returnvalue "FX_repulsor_r" "KeepRelative" "KeepRelative" "KeepRelative" true))

(event Custom|BoostActivate
  ;[0.5 1 1] Boost Active - Idle Color Control and Emitter Enabled
  (Components|Activation|Activate (Variables|Default|GetBoostJet_Right) true)
  (bind _returnvalue (Utilities|Time|SetTimerbyEvent (delegate Custom|BoostCurveChecker) 0.01 true))
  (Variables|Default|SetBooster%20Timer _returnvalue))

(event Custom|BoostCurveChecker
  ;[0.5 1 1] Boost Active - Idle Color Control and Emitter Enabled
  (if (not (> (Animation|Curves|GetCurveValue (Variables|Default|GetIndigoAnimClass) "IndigoBoost") (Variables|Default|GetCurve%20Tolerance)))
    (Utilities|Time|ClearandInvalidateTimerbyHandle (Variables|Default|GetBooster%20Timer))
    (Components|Activation|Deactivate (Variables|Default|GetBoostJet_Idle_Right)))
  (Utilities|FlowControl|DoOnce false
    (:Completed
      (Components|Activation|Activate (Variables|Default|GetBoostJet_Idle_Right) false))))

(event Custom|Boost_Notify_Setup
  ;[0.5 1 1] Boost Setups
  (bind _asfortnite_m_avg_player_anim_blueprint (Utilities|Casting|CastToFortnite_M_Avg_Player_AnimBlueprint (Components|SkeletalMesh|GetAnimInstance (Variables|/Script/Engine.Character|GetMesh (CustomCharacter|GetAssociatedPlayerPawn)))))
  (bind _output_get (Variables|Default|SetIndigoAnimClass _asfortnite_m_avg_player_anim_blueprint))
  (Default|BindEventtoIndigoBoost _output_get (EventDispatchers|CreateEvent :!event Custom|BoostActivate)))

(fn OnRep_OurCurrentWeapon
  (bind _asb_ranged_ashton_indigo_athena (Utilities|Casting|CastToB_Ranged_Ashton_Indigo_Athena (Variables|Default|GetOurCurrentWeapon))
    (:then
      (RangedWeaponDual|BindEventtoOnHandFired _asb_ranged_ashton_indigo_athena (EventDispatchers|CreateEvent :!event Firing))
      (Development|PrintString (Utilities|GetDisplayName _asb_ranged_ashton_indigo_athena) true true "(R=0.290283,G=1.000000,B=0.000000,A=1.000000)"))
    (:CastFailed
      (Development|PrintString "CASTFAILED" true true "(R=1.000000,G=0.000000,B=0.031557,A=1.000000)"))))

(fn ConstructionScript)
