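; Cosmetic part modifier Blueprint for the Bling backpack that selects which VFX component remains once the part finishes its res-in.
; On OnPostResIn it runs the parent logic, then checks whether the associated player pawn is locally controlled.
; Local pawns destroy the LocalBackpackVFXNiagara component.
; Other pawns destroy the idle VFX and attach that Niagara component to the part mesh's backpack_root, snapping location and rotation.
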(event EventOnPostResIn
  (|CallParentFunction)
  (if (Pawn|IsLocallyControlled (CustomCharacter|GetAssociatedPlayerPawn))
    (Components|DestroyComponent (Variables|Default|GetLocalBackpackVFXNiagara))
    (else
      (Components|DestroyComponent (CallFunction|GetIdleVFXComponent))
      (Transformation|AttachComponentToComponent (Variables|Default|GetLocalBackpackVFXNiagara) (CustomCharacter|GetSkeletalMeshForPartOnPlayerPawn) "backpack_root" "SnapToTarget" "SnapToTarget" "KeepWorld" true))))

(fn ConstructionScript)
