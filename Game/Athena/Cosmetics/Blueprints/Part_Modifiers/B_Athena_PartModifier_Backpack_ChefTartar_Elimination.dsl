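; Backpack part modifier for the Chef Tartar cosmetic that makes the back bling react to eliminations.
; On its reactivity event it calls the parent, then gets the owner's skeletal mesh for the part.
; It writes the current game time to a "Timestamp" material scalar and the kill count to a "KillCount" scalar, leaving the visual response to the materials.
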(event EventOnReactivity
  (params (KillCount :type Integer))
  (bind _returnvalue (CustomCharacter|GetSkeletalMeshforPartonOwner))
  (|CallParentFunction KillCount)
  (Rendering|Material|SetScalarParameterValueonMaterials _returnvalue "Timestamp" (Utilities|Time|GetGameTimeinSeconds))
  (Rendering|Material|SetScalarParameterValueonMaterials _returnvalue "KillCount" (Math|Conversions|ToFloat(Integer) KillCount)))

(fn ConstructionScript
  (|CallParentFunction))
