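; Cosmetic part modifier for the Bananas character part that progresses a material effect over the course of a match.
; Each tick outside the front end, it divides game state match time by a fade-in duration and writes it to the "MatchTime" scalar on the pawn's skeletal mesh for this part type.
; Once the ratio exceeds one, it pins the parameter at one and disables its own tick.
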(event EventBeginPlay)

(event Collision|EventActorBeginOverlap
  (params (OtherActor :type Actor%20Object%20Reference)))

(event EventTick
  (params (DeltaSeconds :type Float%20%28single-precision%29))
  (bind _returnvalue (CustomCharacter|GetCharacterPartType))
  (bind _returnvalue_2 (CustomCharacter|GetAssociatedPlayerPawn))
  (bind _returnvalue_3 (Customization|GetSkeletalMeshforPartType _returnvalue_2 _returnvalue :!class /Script/FortniteGame.FortPlayerPawn))
  (bind _amount_of_time_to_complete_fade_in_secs (Variables|Default|GetAmount%20of%20Time%20to%20Complete%20Fade%20in%20Secs))
  (|CallParentFunction DeltaSeconds)
  (bind _asfort_game_state (Utilities|Casting|CastToFortGameState (Game|GetGameState)))
  (if (CustomCosmeticModifier|IsinFrontEnd)
    _
    (elif (> (/ (GameState|GetMatchTime _asfort_game_state) _amount_of_time_to_complete_fade_in_secs) 1.0)
      (Actor|Tick|SetActorTickEnabled false)
      (Rendering|Material|SetScalarParameterValueonMaterials _returnvalue_3 "MatchTime" 1.0)
      (else
        (Rendering|Material|SetScalarParameterValueonMaterials _returnvalue_3 "MatchTime" (/ (GameState|GetMatchTime _asfort_game_state) _amount_of_time_to_complete_fade_in_secs))))))

(fn ConstructionScript
  (|CallParentFunction))
