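; Part modifier Blueprint for the BeaverStage cosmetic that adjusts the character part's material when shown in the frontend.
; After the parent's post-resolution event runs, it checks whether it is in the frontend.
; If so, it sets the "Frontend - AO Base Color Intensity" scalar on the part's skeletal mesh materials to full.
; The construction script is empty.
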(event EventOnPostResIn
  (|CallParentFunction)
  (if (CustomCosmeticModifier|IsinFrontEnd)
    (Rendering|Material|SetScalarParameterValueonMaterials (CustomCharacter|GetSkeletalMeshForPartOnPlayerPawn) "Frontend - AO Base Color Intensity" 1.0)))

(fn ConstructionScript)
