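; Character part modifier for the UnluckyRoll cosmetic that keeps its idle VFX in step with the character mesh's level of detail.
; After res-in, it caches the part's skeletal mesh from the player pawn.
; Every tick, while that mesh is valid, it sets the idle particle component's LOD23 bool to whether the predicted LOD is 2 or higher.
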(event EventTick
  (params (DeltaSeconds :type Float%20%28single-precision%29))
  (bind _pawnmesh (Variables|Default|GetPawnMesh))
  (Utilities|IsValid _pawnmesh
    (:Is%20Valid
      (Effects|Components|ParticleSystem|SetBoolParameter (CallFunction|GetIdleVFXComponent) "LOD23" (>= (Components|SkinnedMesh|GetPredictedLODLevel _pawnmesh) 2)))))

(event EventOnPostResIn
  (|CallParentFunction)
  (bind _asskeletal_mesh_component (Utilities|Casting|CastToSkeletalMeshComponent (CustomCharacter|GetSkeletalMeshForPartOnPlayerPawn)))
  (Variables|Default|SetPawnMesh _asskeletal_mesh_component))

(fn ConstructionScript
  (|CallParentFunction))
