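; Cosmetic part modifier for the Comet character body that drives an idle particle effect, apparently shedding fluff, in response to pawn movement.
; Part changes set the effect's frontend-preview flag, while jumping and landing re-activate it.
; Skydive start runs an effects-quality check and sets a looping timer that keeps re-activating the effect; skydive end clears that timer.
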(event EventOnPawnAndPartChanged
  (|CallParentFunction)
  (if (CustomCosmeticModifier|IsinFrontEnd)
    (Effects|Components|ParticleSystem|SetBoolParameter (CallFunction|GetIdleVFXComponent) "bIsFrontendPreview" true)
    (else
      (Effects|Components|ParticleSystem|SetBoolParameter (CallFunction|GetIdleVFXComponent) "bIsFrontendPreview" false))))

(event EventOnPawnJumped
  (|CallParentFunction)
  (Components|Activation|Activate (CallFunction|GetIdleVFXComponent) true))

(event Custom|Spawn%20Flulff%20while%20Skydiving
  (Components|Activation|Activate (CallFunction|GetIdleVFXComponent) true))

(event EventOnPawnLanded
  (|CallParentFunction)
  (Components|Activation|Activate (CallFunction|GetIdleVFXComponent) true))

(event EventOnBeginSkydiving
  (|CallParentFunction)
  (Utilities|EffectsQualityCheck 0
    (:Cinematic
      _)
    (:Epic
      _)
    (:High
      _))
  (bind _returnvalue (Utilities|Time|SetTimerbyEvent (EventDispatchers|CreateEvent :!event Custom|Spawn%20Flulff%20while%20Skydiving) 0.2 true))
  (Variables|Default|SetDropping%20While%20Diving _returnvalue))

(event EventOnEndSkydiving
  (|CallParentFunction)
  (Utilities|EffectsQualityCheck 0
    (:Cinematic
      _)
    (:Epic
      _)
    (:High
      _))
  (Utilities|Time|ClearandInvalidateTimerbyHandle (Variables|Default|GetDropping%20While%20Diving)))

(fn ConstructionScript)
