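; Cosmetic part modifier that adds glowing eyes and an eye trail to the DarkViking character.
; On post res-in, it runs the parent logic and attaches the left glow, right glow and trail base components to the head socket of the pawn's part skeletal mesh.
; The attachments keep their relative transforms and weld simulated bodies.
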(event EventOnPostResIn
  (bind _returnvalue (CustomCharacter|GetSkeletalMeshForPartOnPlayerPawn))
  (|CallParentFunction)
  (Transformation|AttachComponentToComponent (Variables|Default|GetSM_Eye_Glow_L) _returnvalue "head" "KeepRelative" "KeepRelative" "KeepRelative" true)
  (Transformation|AttachComponentToComponent (Variables|Default|GetSM_Eye_Glow_R) _returnvalue "head" "KeepRelative" "KeepRelative" "KeepRelative" true)
  (Transformation|AttachComponentToComponent (Variables|Default|GetSM_Eye_Trail_Base) _returnvalue "head" "KeepRelative" "KeepRelative" "KeepRelative" true))

(fn ConstructionScript)
