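; Cosmetic part modifier for the Dazzle character part that makes its idle VFX react to eliminations.
; A stat change stores the value as the current elimination count and, when it is at least one, fires the effect trigger.
; The trigger raises the TriggerKill particle parameter, then clears it after a retriggerable delay, so repeated triggers extend the burst.
; After res-in, it also runs a toon frontend LUT replacement.
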(event EventOnPostResIn
  ;[0.007499 0.2122 0.6584] FrontEnd Stage cycling
  (|CallParentFunction)
  (Utilities|ToonFrontendLUTReplace))

(event EventOnStatChanged
  (params (State%20Name :type Name :default "None") (Stat%20Value :type Integer))
  ;[0.272 0.8177 0.1576] Reactivity logic
  (|CallParentFunction State%20Name Stat%20Value)
  (Variables|Default|SetCurrent%20Eliminations Stat%20Value)
  (if (>= (Variables|Default|GetCurrent%20Eliminations) 1)
    (Utilities|FlowControl|Sequence
      (:then_0
        (CallFunction|TriggerFX)))))

(event Custom|Trigger%20FX
  (bind _returnvalue (CallFunction|GetIdleVFXComponent))
  ;[1 0.4948 0.4948] VFX
  (Effects|Components|ParticleSystem|SetBoolParameter _returnvalue "TriggerKill" true)
  (Utilities|FlowControl|RetriggerableDelay (Variables|Default|GetDuration))
  (Effects|Components|ParticleSystem|SetBoolParameter _returnvalue "TriggerKill" false))

(event EventBeginPlay)

(fn ConstructionScript)
