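; Cosmetic part modifier for the Elastic character that derives rim-light settings from its primary and secondary material colors.
; ColorModifications runs on construction and after res-in, reading dynamic materials, storing results and registering a cosmetic-part-applied delegate.
; CalcRimColors lightens each color and derives view-space strength from desaturated brightness, overriding specific preset colors like black and white.
; SetUpdates defines out-of-frontend material overrides for far rim lighting, but nothing here calls it.
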(event EventOnPostResIn
  (|CallParentFunction)
  (Utilities|FlowControl|Delay 0.2)
  (CallFunction|ColorModifications))

(event Custom|ColorModifications
  (bind _returnvalue (CustomCharacter|GetAssociatedPlayerPawn))
  (bind _returnvalue_2 (CustomCharacter|GetSkeletalMeshForPartOnPlayerPawn))
  (bind _enableupdates (Variables|Default|GetEnableUpdates))
  ; ColorModificationsColapsed
  (Variables|Default|SetEnableUpdates true)
  (Utilities|FlowControl|Sequence
    (:then_0
      (Utilities|IsValid _returnvalue_2
        (:Is%20Valid
          (for _array_element (Rendering|Material|GetMaterials _returnvalue_2)
            (bind _asmaterial_instance_dynamic (Utilities|Casting|CastToMaterialInstanceDynamic _array_element))
            (Utilities|FlowControl|Sequence
              (:then_0
                (bind _returnvalue_3 (Rendering|Material|GetVectorParameterValue _asmaterial_instance_dynamic "PrimaryColor" :!class /Script/Engine.MaterialInstanceDynamic))
                (bind (_outcolor _mult _viewspaceshading) (CallFunction|CalcRimColors _returnvalue_3 _enableupdates))
                (Variables|Default|SetprimaryRim _outcolor)
                (Variables|Default|SetprimaryRimMult _mult))
              (:then_1
                (bind _returnvalue_4 (Rendering|Material|GetVectorParameterValue _asmaterial_instance_dynamic "SecondaryColor" :!class /Script/Engine.MaterialInstanceDynamic))
                (bind (_outcolor_2 _mult_2 _viewspaceshading_2) (CallFunction|CalcRimColors _returnvalue_4 _enableupdates))
                (Variables|Default|SetsecondaryRim _outcolor_2)
                (Variables|Default|SetsecondaryRimMult _mult_2))
              (:then_2
                (Variables|Default|SetviewStrength (Math|Float|Max(Float) _viewspaceshading _viewspaceshading_2))))))))
    (:then_1
      (Utilities|IsValid _returnvalue
        (:Is%20Valid
          (bind _delegatehandlecontroller_out (Cosmetics|WhenCosmeticPartApplied _returnvalue (delegate Custom|OnCosmeticPart) true))
          (Variables|Default|SetDelegate%20Handle%20Controller _delegatehandlecontroller_out))))))

(fn CalcRimColors
  (params (inColor :type Linear%20Color%20Structure) (SetImprovements :type Boolean))
  (bind _returnvalue_r (Math|Color|BreakColor (Math|Color|Desaturate(LinearColor) inColor 1.0)))
  (bind _returnvalue (Math|Color|SelectColor "(R=0.313726,G=0.898039,B=1.000000,A=1.000000)" "(R=0.027451,G=0.839216,B=1.000000,A=1.000000)" SetImprovements))
  (LocalVariables|SetrimColor (Math|Color|Lerp(LinearColor) inColor "(R=0.800000,G=0.800000,B=0.800000,A=1.000000)" 0.5))
  (LocalVariables|SetrimMult 1.25)
  (Utilities|FlowControl|Sequence
    (:then_0
      ;[1 1 1] viewStrength - default to desaturated value of linear color.
      (LocalVariables|SetviewSpace (Math|Float|Min(Float) (+ (* _returnvalue_r 1.0) 0.25) 1.0)))
    (:then_1
      ;[1 1 1] black
      (if (Math|Color|NearEqual(LinearColor) inColor "(R=0.020000,G=0.020000,B=0.020000,A=1.000000)" 0.02)
        (LocalVariables|SetrimColor "(R=0.262500,G=0.434000,B=0.875000,A=1.000000)")
        (LocalVariables|SetviewSpace 0.0)))
    (:then_2
      ;[1 1 1] silver
      (if (Math|Color|NearEqual(LinearColor) inColor "(R=0.100000,G=0.100000,B=0.100000,A=1.000000)" 0.02)
        (LocalVariables|SetrimColor "(R=0.262500,G=0.434000,B=0.875000,A=1.000000)")
        (LocalVariables|SetviewSpace 0.0)))
    (:then_3
      ;[1 1 1] platinum
      (if (Math|Color|NearEqual(LinearColor) inColor "(R=0.400000,G=0.400000,B=0.400000,A=1.000000)" 0.02)
        (LocalVariables|SetrimColor _returnvalue)
        (LocalVariables|SetviewSpace 1.0)))
    (:then_4
      ;[1 1 1] white
      (if (Math|Color|NearEqual(LinearColor) inColor "(R=0.800000,G=0.800000,B=0.800000,A=1.000000)" 0.02)
        (LocalVariables|SetrimColor _returnvalue)
        (LocalVariables|SetviewSpace 1.0)))
    (:then_5
      ;[1 1 1] midnight blue
      (if (Math|Color|NearEqual(LinearColor) inColor "(R=0.010000,G=0.030000,B=0.298000,A=1.000000)" 0.02)
        (LocalVariables|SetrimColor "(R=0.400000,G=0.428000,B=0.500000,A=1.000000)")))
    (:then_6
      (return (LocalVariables|GetrimColor) (LocalVariables|GetrimMult) (LocalVariables|GetviewSpace))))
  (LocalVariables|SetrimMult 2.0))

(fn GameTimeFromBPForGlow
  (params (A :type Float%20%28double-precision%29) (B :type Float%20%28double-precision%29))
  (return 0.0))

(fn SetUpdates
  (params (Target :type Mesh%20Component%20Object%20Reference))
  (bind _enableupdates (Variables|Default|GetEnableUpdates))
  (if (not (or (CustomCosmeticModifier|IsinFrontEnd) (Reactive|Frontend|IsCosmeticObjectDisplayedinFrontendContext 0)))
    (Utilities|FlowControl|Sequence
      (:then_0
        (Rendering|Material|SetScalarParameterValueonMaterials Target "EnableElasticUpdates" (Math|Conversions|ToFloat(Boolean) _enableupdates)) ; Enable material overrides
      )
      (:then_2
        (Rendering|Material|SetVectorParameterValueonMaterials Target "elastic_farRimDirection" (Math|Vector|SelectVector "1.500000,2.000000,1.000000" "1.000000,2.000000,2.000000" _enableupdates)) ; Reposition the Camera Space rim light to increase range
      )
      (:then_4
        (Rendering|Material|SetScalarParameterValueonMaterials Target "colorRim_far" (Math|Float|SelectFloat 3.0 2.0 _enableupdates))))))

(fn ConstructionScript
  (CallFunction|ColorModifications))
