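; Evil Cowboy head part modifier that restyles the cosmetic's materials as its stages activate.
; Cosmetic state activation runs Stage 1 or Stage 2, and in the front end Stage 2 also fires two seconds after res-in.
; Each stage, gated by CanExecuteCosmeticEvents, writes its name-keyed color and scalar maps to the part's skeletal mesh materials.
; Scalars apply only when that stage's color map is non-empty.
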(event Custom|Stage%201
  (bind _returnvalue (CustomCharacter|GetAssociatedPlayerPawn))
  (bind _returnvalue_2 (CustomCharacter|GetCharacterPartType))
  (bind _returnvalue_3 (Customization|GetSkeletalMeshforPartType _returnvalue _returnvalue_2 :!class /Script/FortniteGame.FortPlayerPawn))
  (bind _stage_1_material_colors (Variables|Default|GetStage%201%20Material%20Colors))
  (bind _stage_1_material_scalars (Variables|Default|GetStage%201%20Material%20Scalars))
  (Networking|CanExecuteCosmeticEvents
    (:True
      ;[1 1 1] Stage 1 Material Changes for Character Part
      (bind _keys (Utilities|Map|Keys _stage_1_material_colors))
      (if (CallFunction|ArrayIsValid _keys)
        (for _array_element _keys
          (if (Utilities|Name|Equal(Name) _array_element "None")
            (Development|PrintString "What are you thinking?  A Variable called None.  Really?!")
            (else
              (Rendering|Material|SetVectorParameterValueonMaterials _returnvalue_3 _array_element (Math|Conversions|ToVector(LinearColor) (Utilities|Map|Find _stage_1_material_colors _array_element))))))
        (bind _keys_2 (Utilities|Map|Keys _stage_1_material_scalars))
        (if (CallFunction|ArrayIsValid _keys_2)
          (for _array_element_2 _keys_2
            (if (Utilities|Name|Equal(Name) _array_element_2 "None")
              (Development|PrintString "What are you thinking?  A Variable called None.  Really?!")
              (else
                (Rendering|Material|SetScalarParameterValueonMaterials _returnvalue_3 _array_element_2 (Utilities|Map|Find _stage_1_material_scalars _array_element_2))))))))))

(event EventOnPostResIn
  (|CallParentFunction)
  (if (CustomCosmeticModifier|IsinFrontEnd)
    (Utilities|FlowControl|Delay 2.0)
    (CallFunction|Stage2)))

(event EventOnCosmeticStateActivated
  (params (Stat%20Name :type Name :default "None") (Stat%20Value :type Integer) (STAGE%20Num%20Being%20Activated :type Integer))
  (|CallParentFunction Stat%20Name Stat%20Value STAGE%20Num%20Being%20Activated)
  (switch int STAGE%20Num%20Being%20Activated
    (:1
      (CallFunction|Stage1))
    (:2
      (CallFunction|Stage2))))

(event Custom|Stage%202
  (bind _returnvalue (CustomCharacter|GetAssociatedPlayerPawn))
  (bind _returnvalue_2 (CustomCharacter|GetCharacterPartType))
  (bind _returnvalue_3 (Customization|GetSkeletalMeshforPartType _returnvalue _returnvalue_2 :!class /Script/FortniteGame.FortPlayerPawn))
  (bind _stage_2_material_colors (Variables|Default|GetStage%202%20Material%20Colors))
  (bind _stage_2_material_scalars (Variables|Default|GetStage%202%20Material%20Scalars))
  (Networking|CanExecuteCosmeticEvents
    (:True
      ;[1 1 1] Stage 1 Material Changes for Character Part
      (bind _keys (Utilities|Map|Keys _stage_2_material_colors))
      (if (CallFunction|ArrayIsValid _keys)
        (for _array_element _keys
          (if (Utilities|Name|Equal(Name) _array_element "None")
            (Development|PrintString "What are you thinking?  A Variable called None.  Really?!")
            (else
              (Rendering|Material|SetVectorParameterValueonMaterials _returnvalue_3 _array_element (Math|Conversions|ToVector(LinearColor) (Utilities|Map|Find _stage_2_material_colors _array_element))))))
        (bind _keys_2 (Utilities|Map|Keys _stage_2_material_scalars))
        (if (CallFunction|ArrayIsValid _keys_2)
          (for _array_element_2 _keys_2
            (if (Utilities|Name|Equal(Name) _array_element_2 "None")
              (Development|PrintString "What are you thinking?  A Variable called None.  Really?!")
              (else
                (Rendering|Material|SetScalarParameterValueonMaterials _returnvalue_3 _array_element_2 (Utilities|Map|Find _stage_2_material_scalars _array_element_2))))))))))

(fn ArrayIsValid
  (params (Name%20Array :type Array%20of%20Names))
  (return (> (Utilities|Array|Length Name%20Array) 0)))

(fn ConstructionScript
  (|CallParentFunction))
