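; Character part modifier for the F_MED_Sheath cosmetic that animates a goggle transition when the wearer aims down sights.
; On targeting or first-person camera changes, it checks the owning pawn's targeting state and plays or reverses the ADS_Goggles timeline.
; The timeline drives the GoggleTransition scalar on the materials of the part's skeletal mesh.
; The Aiming function only casts and checks targeting, with an empty branch.
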(event EventOnTargetingOrFirstPersonCameraChanged
  (|CallParentFunction)
  (if (Variables|/Script/FortniteGame.FortPlayerPawn|GetbIsTargeting (CustomCharacter|GetAssociatedPlayerPawn))
    (bind (_direction _newtrack_0) (Timeline|ADS_Goggles Play)
      (:Update
        (Rendering|Material|SetScalarParameterValueonMaterials (CustomCharacter|GetSkeletalMeshForPartOnPlayerPawn) "GoggleTransition" _newtrack_0)))
    (else
      (Timeline|ADS_Goggles Reverse))))

(fn Aiming
  (params (Object :type Object%20Reference))
  (bind _asfort_player_pawn_athena (Utilities|Casting|CastToFortPlayerPawnAthena Object))
  (if (Variables|/Script/FortniteGame.FortPlayerPawn|GetbIsTargeting _asfort_player_pawn_athena)
    _))

(fn ConstructionScript
  (|CallParentFunction))
