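; Character part modifier that hides a cosmetic's hand idle VFX when the pawn or its weapon carries the Animation.CharacterParts.HideFX.Hands tag.
; Tag add/remove tasks on the pawn and weapon-change checks set separate hidden flags, with an extra player-granted tag check for frontend display.
; UpdateHidden ORs both flags and, only on change, sets the VFX's bisHandHidden parameter without toggling component visibility.
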(event EventOnNewPlayerWeaponType
  (params (NewType :type EFortItemType%20Enum :default "WorldItem") (OldType :type EFortItemType%20Enum :default "WorldItem"))
  (|CallParentFunction NewType OldType) ; This node is disabled and will not be called.\nDrag off pins to build functionality.
)

(event EventOnWeaponChanged
  (params (NewWeapon :type Fort%20Weapon%20Object%20Reference) (OldWeapon :type Fort%20Weapon%20Object%20Reference))
  (|CallParentFunction NewWeapon OldWeapon)
  (CallFunction|SetVisibilityFromWeapon NewWeapon))

(event EventOnPawnAndPartChanged
  (bind _returnvalue (CustomCharacter|GetAssociatedPlayerPawn))
  (|CallParentFunction)
  (Utilities|IsValid _returnvalue
    (:Is%20Valid
      (Utilities|FlowControl|Sequence
        (:then_0
          (Ability|Tasks|WaitGameplayTagAddToActor _returnvalue "(TagName=\"Animation.CharacterParts.HideFX.Hands\")"
            (:Added
              (Variables|Default|SetHidden%20By%20Loose%20Tag true)
              (CallFunction|UpdateHidden))))
        (:then_1
          (Ability|Tasks|WaitGameplayTagRemoveFromActor _returnvalue "(TagName=\"Animation.CharacterParts.HideFX.Hands\")"
            (:Removed
              (Variables|Default|SetHidden%20By%20Loose%20Tag false)
              (CallFunction|UpdateHidden))))
        (:then_2
          (CallFunction|SetVisibilityFromWeapon (Weapon|GetCurrentWeapon _returnvalue)))))))

(event Custom|Update%20Hidden
  (Variables|Default|SetWas%20Hidden (Variables|Default|GetIs%20Hidden))
  (Variables|Default|SetIs%20Hidden (or (Variables|Default|GetHidden%20By%20Weapon) (Variables|Default|GetHidden%20By%20Loose%20Tag)))
  (if (!= (Variables|Default|GetIs%20Hidden) (Variables|Default|GetWas%20Hidden))
    (Default|SetVisibilityOnHandFX (Utilities|Array|MakeArray (CallFunction|GetIdleVFXComponent)) true (not (Variables|Default|GetIs%20Hidden)))))

(fn SetVisibilityfromWeapon
  (params (New%20Weapon :type Fort%20Weapon%20Object%20Reference))
  (Utilities|IsValid New%20Weapon
    (:Is%20Valid
      (Variables|Default|SetHidden%20By%20Weapon (or (and (Reactive|Frontend|IsCosmeticObjectDisplayedinFrontendContext 0) (Tags|WeaponDataHasMatchingPlayerGrantedGameplayTag New%20Weapon "(TagName=\"Animation.CharacterParts.HideFX.Hands\")")) (Tags|WeaponDataHasMatchingGameplayTag New%20Weapon "(TagName=\"Animation.CharacterParts.HideFX.Hands\")")))
      (CallFunction|UpdateHidden))))

(fn SetVisibilityonHandFX
  (params (FX%20Systems%20associated%20with%20Hands :type Array%20of%20FXSystem%20Component%20Object%20References) (Only%20Set%20Niagara%20Bool%20Variable :type Boolean) (Visible :type Boolean))
  (bind _returnvalue (not Visible))
  (Effects|Components|ParticleSystem|SetBoolParameter FX%20Systems%20associated%20with%20Hands "bisHandHidden" _returnvalue)
  (if (not Only%20Set%20Niagara%20Bool%20Variable)
    (Development|SetHiddeninGame FX%20Systems%20associated%20with%20Hands _returnvalue true)))

(fn ConstructionScript
  (|CallParentFunction))
