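; Cosmetic part modifier for Impulse that adjusts a character's morph override data by variant and reapplies morph material parameters on post res-in.
; It finds the cosmetic data component on the player pawn, falling back to a player mannequin.
; When the Impulse meta tag (a country tag) matches a configured tag, it rewrites the morph override property using before/after priority tags.
; It then reapplies morph material parameters with ExtraPart and Body.
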(event EventOnPostResIn
  (bind _returnvalue (Actor|GetOwner))
  (bind _returnvalue_2 (CustomCharacter|GetAssociatedPlayerPawn))
  (|CallParentFunction)
  (CallFunction|SetupDynamicMaterialParameterforVariants)
  (Utilities|IsValid _returnvalue_2
    (:Is%20Valid
      (Cosmetics|ReapplyMorphMaterialParameters _returnvalue_2 "ExtraPart" "Body"))
    (:Is%20Not%20Valid
      (Utilities|IsValid _returnvalue
        (:Is%20Valid
          (bind _asfort_player_mannequin (Utilities|Casting|CastToFortPlayerMannequin _returnvalue))
          (Class|FortPlayerMannequin|ReapplyMorphMaterialParameters _asfort_player_mannequin "ExtraPart" "Body"))))))

(event EventOnCurieStateDetached
  (params (Tag :type Gameplay%20Tag%20Structure%20%28by%20ref%29)))

(event EventOnBeginSkydiving
  (|CallParentFunction) ; This node is disabled and will not be called.\nDrag off pins to build functionality.
)

(event EventOnEndSkydiving
  (|CallParentFunction) ; This node is disabled and will not be called.\nDrag off pins to build functionality.
)

(event EventOnEnterVehicle
  (params (VehicleEntered :type Fort%20Vehicle%20Interface%20Interface%20%28by%20ref%29))
  (|CallParentFunction VehicleEntered) ; This node is disabled and will not be called.\nDrag off pins to build functionality.
)

(event EventOnExitVehicle
  (|CallParentFunction) ; This node is disabled and will not be called.\nDrag off pins to build functionality.
)

(event EventOnTargetingOrFirstPersonCameraChanged)

(event EventOnHitPawn
  (params (SelfActor :type Actor%20Object%20Reference) (OtherActor :type Actor%20Object%20Reference) (NormalImpulse :type Vector :default "0, 0, 0") (Hit :type Hit%20Result%20Structure%20%28by%20ref%29)))

(fn GetOverridePropertyFromVariant
  (params (CosmeticDataComponent :type Cosmetic%20Data%20Component%20%28Experimental%29%20Object%20Reference))
  (Utilities|IsValid CosmeticDataComponent
    (:Is%20Valid
      (bind _outproperty (Class|CosmeticDataComponent%28Experimental%29|FindProperty CosmeticDataComponent "(TagName=\"LoadoutSlot.Character\")" (Variables|Default|GetTagOnPropertyBeforePriority))
        (:True
          (bind _output_get (LocalVariables|SetPropertyBeforePriority _outproperty))
          (bind _structout (Utilities|Struct|SetmembersinFortCosmeticPropertyMorphOverrideData _output_get (Variables|Default|GetTagOnPropertyAfterPriority)))
          (return _structout))))))

(fn SetupDynamicMaterialParameterForVariants
  (bind _returnvalue (Actor|GetOwner))
  (bind _returnvalue_2 (CustomCharacter|GetAssociatedPlayerPawn))
  (bind _returnvalue_3 (Cosmetic|GetCosmeticDataComponent _returnvalue_2))
  (Utilities|IsValid _returnvalue_2
    (:Is%20Valid
      (Utilities|IsValid _returnvalue_3
        (:Is%20Valid
          (LocalVariables|SetCosmeticDataComponent _returnvalue_3))))
    (:Is%20Not%20Valid
      (Utilities|IsValid _returnvalue
        (:Is%20Valid
          (bind _asfort_player_mannequin (Utilities|Casting|CastToFortPlayerMannequin _returnvalue))
          (bind _output_get (LocalVariables|SetPlayerMannequin _asfort_player_mannequin))
          (Utilities|IsValid (Variables|/Script/FortniteGame.FortPlayerMannequin|GetCosmeticDataComponent _output_get)
            (:Is%20Valid
              (LocalVariables|SetCosmeticDataComponent (Variables|/Script/FortniteGame.FortPlayerMannequin|GetCosmeticDataComponent _output_get))))))))
  (bind _outproperty (Class|CosmeticDataComponent%28Experimental%29|FindProperty (LocalVariables|GetCosmeticDataComponent) "(TagName=\"LoadoutSlot.Character\")" "(TagName=\"Cosmetics.MetaTags.Impulse\")")
    (:True
      (bind _output_get_2 (LocalVariables|SetCountryTagProperty _outproperty))
      (bind _output_get_3 (LocalVariables|SetCountryTag (Utilities|Struct|BreakCosmeticPropertyGameplayTag _output_get_2)))
      (if (== _output_get_3 (Variables|Default|GetTagToSwitchPriorityOn))
        (bind _fortcosmeticproperty_morphoverridedata (CallFunction|GetOverridePropertyfromVariant (LocalVariables|GetCosmeticDataComponent)))
        (Class|CosmeticDataComponent%28Experimental%29|AddOrOverrideProperty (LocalVariables|GetCosmeticDataComponent) "(TagName=\"LoadoutSlot.Character\")" _fortcosmeticproperty_morphoverridedata)))))

(fn ConstructionScript
  (|CallParentFunction))
