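; Cosmetic part modifier for the Kepler body that drives its idle VFX particle parameters from pawn movement and events.
; Movement updates push the burst flag and write JumpTime twice, so ReactiveSpawnRate overrides the counter of jump-force updates.
; Landing pulses SpawnBurst briefly.
; Emote end or a positive stat change plays a timeline that animates ReactiveSpawnRate.
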(event EventOnCharacterMovementUpdated
  (params (DeltaSeconds :type Float%20%28single-precision%29) (OldLocation :type Vector :default "0, 0, 0") (OldVelocity :type Vector :default "0, 0, 0"))
  (bind _returnvalue (CallFunction|GetIdleVFXComponent))
  (|CallParentFunction DeltaSeconds OldLocation OldVelocity)
  (Effects|Components|ParticleSystem|SetFloatParameter _returnvalue "JumpTime" (Variables|Default|GetJumpCounter))
  (Effects|Components|ParticleSystem|SetBoolParameter _returnvalue "SpawnBurst" (Variables|Default|GetSpawnFXBurst))
  (Effects|Components|ParticleSystem|SetFloatParameter _returnvalue "JumpTime" (Variables|Default|GetReactiveSpawnRate))
  (if (Character|IsJumpProvidingForce (CustomCharacter|GetAssociatedPlayerPawn))
    (Math|Float|IncrementFloat (Variables|Default|GetJumpCounter))
    (else
      (Variables|Default|SetJumpCounter 0.0))))

(event EventOnPawnLanded
  (|CallParentFunction)
  (Variables|Default|SetSpawnFXBurst true)
  (Utilities|FlowControl|Delay 0.1)
  (Variables|Default|SetSpawnFXBurst false)
  (Effects|Components|ParticleSystem|SetBoolParameter (CallFunction|GetIdleVFXComponent) "SpawnBurst" (Variables|Default|GetSpawnFXBurst)))

(event EventOnEndEmote
  (params (MontageItemDef :type Fort%20Item%20Definition%20Object%20Reference) (PawnEmoting :type Fort%20Pawn%20Object%20Reference))
  (|CallParentFunction MontageItemDef PawnEmoting)
  (bind (_direction _spawnrate) (Timeline|Timeline_0 PlayFromStart)
    (:Update
      (Variables|Default|SetReactiveSpawnRate _spawnrate))))

(event EventOnStatChanged
  (params (State%20Name :type Name :default "None") (Stat%20Value :type Integer))
  (|CallParentFunction State%20Name Stat%20Value)
  (if (> Stat%20Value 0)
    (Timeline|Timeline_0 PlayFromStart)))

(fn ConstructionScript
  (|CallParentFunction))
