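; Part modifier for the LentilAnise backpack cosmetic that makes its eyes blink periodically through material parameters.
; After res-in it starts a looping timer: a fixed interval for frontend display, or the blink cycle time variable in game.
; Each blink stamps the current game time and blink length onto the part's skeletal mesh materials.
; In game, a small random delay, whose range extends below zero, precedes each blink.
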(event Custom|FrontEnd%20Cycle
  (Rendering|Material|SetScalarParameterValueonMaterials (CustomCharacter|GetSkeletalMeshForPartOnPlayerPawn) "TimestampEyeBlink" (Utilities|Time|GetGameTimeinSeconds))
  (Rendering|Material|SetScalarParameterValueonMaterials (CustomCharacter|GetSkeletalMeshForPartOnPlayerPawn) "TransitionLengthEyeBlink" (Variables|Default|GetBlinkLength)))

(event EventOnPostResIn
  (|CallParentFunction)
  (if (Reactive|Frontend|IsCosmeticObjectDisplayedinFrontendContext 0)
    (Utilities|Time|SetTimerbyEvent (EventDispatchers|CreateEvent :!event Custom|FrontEnd%20Cycle) 1e+01 true)
    (else
      (Utilities|Time|SetTimerbyEvent (EventDispatchers|CreateEvent :!event Custom|GameCycle) (Variables|Default|GetBlinkCycletime) true))))

(event Custom|GameCycle
  (bind _output_get (Variables|Default|SetRandomDelay (Math|Random|RandomFloatinRange -0.25 0.25)))
  (Utilities|FlowControl|Delay _output_get)
  (Rendering|Material|SetScalarParameterValueonMaterials (CustomCharacter|GetSkeletalMeshForPartOnPlayerPawn) "TimestampEyeBlink" (Utilities|Time|GetGameTimeinSeconds))
  (Rendering|Material|SetScalarParameterValueonMaterials (CustomCharacter|GetSkeletalMeshForPartOnPlayerPawn) "TransitionLengthEyeBlink" (Variables|Default|GetBlinkLength)))

(fn Kills)

(fn ConstructionScript
  (|CallParentFunction))
