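; Cosmetic character part modifier for the MED Haunt outfit that attaches a Haunt Chest component to the player's body.
; After the parent's post-res-in logic runs, it attaches that component to the part skeletal mesh on the player pawn at spine_05.
; Location and scale snap to the bone, rotation stays relative, and simulated bodies are welded.
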(event EventOnPostResIn
  (|CallParentFunction)
  (Transformation|AttachComponentToComponent (Variables|Default|GetHaunt%20Chest) (CustomCharacter|GetSkeletalMeshForPartOnPlayerPawn) "spine_05" "SnapToTarget" "KeepRelative" "SnapToTarget" true))

(fn ConstructionScript)
