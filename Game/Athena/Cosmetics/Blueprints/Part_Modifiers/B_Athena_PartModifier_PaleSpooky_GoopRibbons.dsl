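; Cosmetic part modifier for the PaleSpooky character that controls its goop ribbons and idle VFX.
; On weapon change and part application, the hidden-hand check, then the hidden-foot check, sets mesh material slot 1 to ribbons or invisible.
; After res-in in a frontend context, it binds pawn material-override events to hide and restore the ribbons.
; Skydiving hides the idle VFX until it ends.
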(event EventOnBeginSkydiving
  (Development|SetHiddeninGame (CallFunction|GetIdleVFXComponent) true false))

(event EventOnWeaponChanged
  (params (NewWeapon :type Fort%20Weapon%20Object%20Reference) (OldWeapon :type Fort%20Weapon%20Object%20Reference))
  (bind _returnvalue (CustomCharacter|GetSkeletalMeshForPartOnPlayerPawn))
  (bind _returnvalue_2 (CustomCharacter|GetSkeletalMeshForPartOnPlayerPawn))
  (|CallParentFunction NewWeapon OldWeapon)
  (if (Variables|HiddenPartCheck|GetbIsHandHidden)
    (Rendering|Material|SetMaterial _returnvalue 1 (Variables|Default|GetInvisibleMaterial))
    (else
      (Rendering|Material|SetMaterial _returnvalue 1 (Variables|Default|GetGoopRibbonsMaterial))))
  (if (Variables|HiddenPartCheck|GetbIsFootHidden)
    (Rendering|Material|SetMaterial _returnvalue_2 1 (Variables|Default|GetInvisibleMaterial))
    (else
      (Rendering|Material|SetMaterial _returnvalue_2 1 (Variables|Default|GetGoopRibbonsMaterial)))))

(event EventOnEndSkydiving
  (Development|SetHiddeninGame (CallFunction|GetIdleVFXComponent) false false))

(event EventOnCosmeticPartAppliedPostResIn
  (params (Part :type Custom%20Character%20Part%20Object%20Reference))
  (bind _returnvalue (CustomCharacter|GetSkeletalMeshForPartOnPlayerPawn))
  (bind _returnvalue_2 (CustomCharacter|GetSkeletalMeshForPartOnPlayerPawn))
  (|CallParentFunction Part)
  (if (Variables|HiddenPartCheck|GetbIsHandHidden)
    (Rendering|Material|SetMaterial _returnvalue 1 (Variables|Default|GetInvisibleMaterial))
    (else
      (Rendering|Material|SetMaterial _returnvalue 1 (Variables|Default|GetGoopRibbonsMaterial))))
  (if (Variables|HiddenPartCheck|GetbIsFootHidden)
    (Rendering|Material|SetMaterial _returnvalue_2 1 (Variables|Default|GetInvisibleMaterial))
    (else
      (Rendering|Material|SetMaterial _returnvalue_2 1 (Variables|Default|GetGoopRibbonsMaterial)))))

(event EventOnPostResIn
  (bind _returnvalue (CustomCharacter|GetAssociatedPlayerPawn))
  (|CallParentFunction)
  (if (Reactive|Frontend|IsCosmeticObjectDisplayedinFrontendContext 0)
    (EventDispatchers|AssignOnMaterialOverrideApplied _returnvalue (delegate Custom|HideGoopRibbons))
    (EventDispatchers|AssignOnMaterialOverrideCleared _returnvalue (delegate Custom|ShowGoopRibbons))))

(event Custom|HideGoopRibbons
  (params (AffectedPawn :type Fort%20Player%20Pawn%20Object%20Reference) (AppliedOverride :type Fort%20Pawn%20Material%20Override%20Structure%20%28by%20ref%29))
  (Rendering|Material|SetMaterial (CustomCharacter|GetSkeletalMeshForPartOnPlayerPawn) 1 (Variables|Default|GetInvisibleMaterial)))

(event Custom|ShowGoopRibbons
  (params (AffectedPawn :type Fort%20Player%20Pawn%20Object%20Reference))
  (Rendering|Material|SetMaterial (CustomCharacter|GetSkeletalMeshForPartOnPlayerPawn) 1 (Variables|Default|GetGoopRibbonsMaterial)))

(fn ConstructionScript)
