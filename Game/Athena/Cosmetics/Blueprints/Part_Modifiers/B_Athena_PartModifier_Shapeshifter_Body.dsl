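; Cosmetic part modifier for the Shapeshifter outfit body, driving a shark-tail transformation when swimming, a Medusa head transformation when aiming, and tattoo glow changes.
; After res-in and a delay, it checks variant channels and binds to Body and Face AnimBP events.
; Tail events spawn Niagara effects using water surface data, faked in frontend, and a timeline animates tail reveal.
; Skydiving, swimming and targeting recolor body gem materials, with a helper selecting the color to restore.
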(event EventOnBeginSkydiving
  (bind _returnvalue (CustomCharacter|GetAssociatedPlayerPawn))
  (bind _returnvalue_2 (Customization|GetSkeletalMeshforPartType _returnvalue "Body" :!class /Script/FortniteGame.FortPlayerPawn))
  ;[1 0.4568 0 0.55] Skydiving
  (|CallParentFunction)
  (Utilities|FlowControl|DoOnce false
    (:Completed
      (Rendering|Material|SetVectorParameterValueonMaterials _returnvalue_2 "Gem Slight Glow Color" (Math|Conversions|ToVector(LinearColor) (Variables|Default|GetSkydiving%20Tattoo)))
      (Rendering|Material|SetScalarParameterValueonMaterials _returnvalue_2 "GemColorTransition" 0.5))))

(event EventOnEndSkydiving
  (bind _returnvalue (CustomCharacter|GetAssociatedPlayerPawn))
  (bind _returnvalue_2 (Customization|GetSkeletalMeshforPartType _returnvalue "Body" :!class /Script/FortniteGame.FortPlayerPawn))
  ;[1 0.4568 0 0.55] Skydiving
  (|CallParentFunction)
  (Utilities|FlowControl|Sequence
    (:then_0
      (Utilities|FlowControl|DoOnce false
        (:Completed
          (Rendering|Material|SetVectorParameterValueonMaterials _returnvalue_2 "Gem Slight Glow Color" (Math|Conversions|ToVector(LinearColor) (Variables|Default|GetSkydiving%20Tattoo)))
          (Rendering|Material|SetScalarParameterValueonMaterials _returnvalue_2 "GemColorTransition" 0.5))))
    (:then_1
      (Rendering|Material|SetVectorParameterValueonMaterials _returnvalue_2 "Gem Slight Glow Color" (Math|Conversions|ToVector(LinearColor) (Variables|Default|GetDefault%20Tattoo))))))

(event Custom|Tail%20Activate
  (bind _returnvalue (CustomCharacter|GetAssociatedPlayerPawn))
  (bind _returnvalue_2 (Customization|GetSkeletalMeshforPartType _returnvalue "MiscOrTail" :!class /Script/FortniteGame.FortPlayerPawn))
  (bind _returnvalue_3 (Transformation|GetSocketLocation _returnvalue_2 "pelvis"))
  (bind _returnvalue_4 (CustomCharacter|GetAssociatedPlayerPawn))
  (bind _returnvalue_5 (Customization|GetSkeletalMeshforPartType _returnvalue_4 "Body" :!class /Script/FortniteGame.FortPlayerPawn))
  (bind _returnvalue_6 (CustomCharacter|GetAssociatedPlayerPawn))
  (bind _returnvalue_7 (Customization|GetSkeletalMeshforPartType _returnvalue_6 "Body" :!class /Script/FortniteGame.FortPlayerPawn))
  (bind _returnvalue_8 (CustomCharacter|GetAssociatedPlayerPawn))
  (bind _returnvalue_9 (Customization|GetSkeletalMeshforPartType _returnvalue_8 "MiscOrTail" :!class /Script/FortniteGame.FortPlayerPawn))
  (bind _returnvalue_10 (CustomCharacter|GetAssociatedPlayerPawn))
  ;[0.02129 0.7064 1] Shark Transformation - Swimming
  (Utilities|FlowControl|DoOnce false
    (:Completed
      (if (CustomCosmeticModifier|IsinFrontEnd)
        (Variables|Default|SetisSwimming true)
        (bind _returnvalue_11 (PartModifier|SpawnSystem|SpawnSystemAttachedPartModifier "/Game/Characters/Player/Female/Medium/Bodies/F_MED_Shapeshifter/FX/NS_Shapeshifter_Charm_ChangeFX.NS_Shapeshifter_Charm_ChangeFX" _returnvalue_2 "None" (Math|Transform|InverseTransformLocation (Transformation|GetWorldTransform _returnvalue_2) _returnvalue_3) "0, 0, 0" "KeepRelativeOffset" true true true "None" false)
          (:Success
            (Niagara|SetNiagaraVariable(Vector3) _returnvalue_11 "WaterPlaneLocation" _returnvalue_3)
            (Niagara|SetNiagaraVariable(Vector3) _returnvalue_11 "WaterPlaneNormal" "0, 0,1.000000")
            (Niagara|SetNiagaraVariable(Float) _returnvalue_11 "WaterBodyIndex" 1.0)
            (Niagara|SetNiagaraVariable(Float) _returnvalue_11 "WaterDepth" 2e+01)
            (Niagara|SetNiagaraSkeletalMeshComponent _returnvalue_11 "SkeletalMesh" (Customization|GetSkeletalMeshforPartType (CustomCharacter|GetAssociatedPlayerPawn) "MiscOrTail" :!class /Script/FortniteGame.FortPlayerPawn))
            (Components|Activation|Activate _returnvalue_11 false))
          (:Failure
            _))
        (bind (_direction _newtrack_0 _glow) (Timeline|Timeline_0 PlayFromStart)
          (:Update
            (Rendering|Material|SetScalarParameterValueonMaterials _returnvalue_9 "Reveal" _newtrack_0)
            (Rendering|Material|SetScalarParameterValueonMaterials _returnvalue_9 "Emissive_BrightnessMin" (Math|Float|Lerp 0.0 0.05 _glow))
            (Rendering|Material|SetScalarParameterValueonMaterials _returnvalue_9 "Emissive_BrightnessMax" (Math|Float|Lerp 0.0 2.0 _glow))
            (Rendering|Material|SetVectorParameterValueonMaterials _returnvalue_9 "ObjectPivotVelocity" (Transformation|GetComponentVelocity (Customization|GetSkeletalMeshforPartType (CustomCharacter|GetAssociatedPlayerPawn) "MiscOrTail" :!class /Script/FortniteGame.FortPlayerPawn)))
            (Rendering|Material|SetVectorParameterValueonMaterials _returnvalue_5 "Gem Slight Glow Color" (Math|Conversions|ToVector(LinearColor) (Variables|Default|GetSwimming%20Tattoo)))
            (Rendering|Material|SetScalarParameterValueonMaterials _returnvalue_5 "GemColorTransition" 1.0))
          (:Finished
            (Utilities|FlowControl|DoOnce true
              (:Completed
                (Variables|Default|SetisSwimming false)
                (PartModifier|SpawnSystem|SpawnSystemAttachedPartModifier "/Game/Characters/Player/Female/Medium/Bodies/F_MED_Shapeshifter/FX/NS_Shapeshifter_Charm_RevertFX.NS_Shapeshifter_Charm_RevertFX" (Customization|GetSkeletalMeshforPartType (CustomCharacter|GetAssociatedPlayerPawn) "Body" :!class /Script/FortniteGame.FortPlayerPawn) "pelvis" "0, 0, 0" "0, 0, 0" "KeepRelativeOffset" true true true "None" false
                  (:Success
                    _)
                  (:Failure
                    _))
                (Rendering|Material|SetVectorParameterValueonMaterials _returnvalue_7 "Gem Slight Glow Color" (Math|Conversions|ToVector(LinearColor) (Variables|Default|GetDefault%20Tattoo)))
                (Rendering|Material|SetScalarParameterValueonMaterials _returnvalue_7 "GemColorTransition" 0.0)))))
        (else
          (Variables|Default|SetisSwimming true)
          (bind _outwatersurfaceinfo (WaterBody|GetWaterSurfaceInfoStruct (WaterBody|GetCurrentWaterBody _returnvalue_10) (Transformation|GetActorLocation _returnvalue_10) true false false))
          (bind (_outwatersurfaceinfo_waterplanelocation _outwatersurfaceinfo_waterplanenormal _outwatersurfaceinfo_watersurfacelocation _outwatersurfaceinfo_waterdepth _outwatersurfaceinfo_waterbodyidx) (Utilities|Struct|BreakWaterSurfaceInfo _outwatersurfaceinfo))
          (bind _returnvalue_12 (PartModifier|SpawnSystem|SpawnSystemAttachedPartModifier "/Game/Characters/Player/Female/Medium/Bodies/F_MED_Shapeshifter/FX/NS_Shapeshifter_Charm_ChangeFX.NS_Shapeshifter_Charm_ChangeFX" (Customization|GetSkeletalMeshforPartType _returnvalue_10 "MiscOrTail" :!class /Script/FortniteGame.FortPlayerPawn) "None" (Math|Transform|InverseTransformLocation (Transformation|GetWorldTransform (Customization|GetSkeletalMeshforPartType _returnvalue_10 "MiscOrTail" :!class /Script/FortniteGame.FortPlayerPawn)) _outwatersurfaceinfo_waterplanelocation) "0, 0, 0" "KeepRelativeOffset" true true true "None" false)
            (:Success
              (Niagara|SetNiagaraVariable(Vector3) _returnvalue_12 "WaterPlaneLocation" _outwatersurfaceinfo_waterplanelocation)
              (Niagara|SetNiagaraVariable(Vector3) _returnvalue_12 "WaterPlaneNormal" _outwatersurfaceinfo_waterplanenormal)
              (Niagara|SetNiagaraVariable(Float) _returnvalue_12 "WaterBodyIndex" (Math|Conversions|ToFloat(Integer) _outwatersurfaceinfo_waterbodyidx))
              (Niagara|SetNiagaraVariable(Float) _returnvalue_12 "WaterDepth" _outwatersurfaceinfo_waterdepth)
              (Niagara|SetNiagaraSkeletalMeshComponent _returnvalue_12 "SkeletalMesh" (Customization|GetSkeletalMeshforPartType (CustomCharacter|GetAssociatedPlayerPawn) "MiscOrTail" :!class /Script/FortniteGame.FortPlayerPawn))
              (Components|Activation|Activate _returnvalue_12 false))
            (:Failure
              _)))))))

(event EventOnPostResIn
  ;[1 1 1] Setup Transformation Calls
  (|CallParentFunction)
  (Utilities|FlowControl|Delay 1.0)
  (Utilities|FlowControl|Sequence
    (:then_0
      (bind _index (CosmeticLoadout|GetVariantChannelActiveIndexforSlotActor "(TagName=\"LoadoutSlot.Character\")" (CustomCharacter|GetAssociatedPlayerPawn) "(TagName=\"Cosmetics.Variant.Channel.Particle\")" (Utilities|Array|MakeArray "(TagName=\"Cosmetics.Variant.Property.Particle1\")" "(TagName=\"Cosmetics.Variant.Property.Particle2\")")))
      (switch int _index
        (:0
          (bind _asf_med_shapeshifter_anim_bp (Utilities|Casting|CastToF_MED_Shapeshifter_AnimBP (Components|SkeletalMesh|GetAnimInstance (Customization|GetSkeletalMeshforPartType (CustomCharacter|GetAssociatedPlayerPawn) "Body" :!class /Script/FortniteGame.FortPlayerPawn)))
            (:then
              (Default|BindEventtoTailOn _asf_med_shapeshifter_anim_bp (EventDispatchers|CreateEvent :!event Custom|Tail%20Activate))
              (Default|BindEventtoTailOff _asf_med_shapeshifter_anim_bp (EventDispatchers|CreateEvent :!event Custom|Tail%20Deactivate)))
            (:CastFailed
              (Development|PrintString "Cast to AnimBP FAILED - Shark Form Disabled"))))))
    (:then_1
      (bind _index_2 (Class|FortCosmeticFunctionLibrary|%28Deprecated%29GetVariantChannelActiveIndexForSlot "(TagName=\"LoadoutSlot.Character\")" (Variables|/Script/FortniteGame.FortPlayerPawn|GetCosmeticLoadout (CustomCharacter|GetAssociatedPlayerPawn)) "(TagName=\"Cosmetics.Variant.Channel.Parts\")" (Utilities|Array|MakeArray "(TagName=\"Cosmetics.Variant.Property.Stage1\")" "(TagName=\"Cosmetics.Variant.Property.Stage2\")")))
      (switch int _index_2
        (:0
          (bind _asf_med_shapeshifter_medusa_anim_bp (Utilities|Casting|CastToF_MED_Shapeshifter_Medusa_AnimBP (Components|SkeletalMesh|GetAnimInstance (Customization|GetSkeletalMeshforPartType (CustomCharacter|GetAssociatedPlayerPawn) "Face" :!class /Script/FortniteGame.FortPlayerPawn)))
            (:then
              (EventDispatchers|BindEventtoActivateTargeting _asf_med_shapeshifter_medusa_anim_bp (EventDispatchers|CreateEvent :!event Custom|Snakes)))
            (:CastFailed
              (Development|PrintString "Cast to 'Dusa AnimBP FAILED - 'Dusa Form Disabled"))))))
    (:then_2
      (Rendering|Material|SetScalarParameterValueonMaterials (Customization|GetSkeletalMeshforPartType (CustomCharacter|GetAssociatedPlayerPawn) "MiscOrTail" :!class /Script/FortniteGame.FortPlayerPawn) "Reveal" 0.0))))

(event Custom|Tail%20Deactivate
  (bind _returnvalue (CustomCharacter|GetAssociatedPlayerPawn))
  (bind _returnvalue_2 (Customization|GetSkeletalMeshforPartType _returnvalue "MiscOrTail" :!class /Script/FortniteGame.FortPlayerPawn))
  (bind _returnvalue_3 (Transformation|GetSocketLocation _returnvalue_2 "pelvis"))
  (bind _returnvalue_4 (CustomCharacter|GetAssociatedPlayerPawn))
  (bind _returnvalue_5 (Customization|GetSkeletalMeshforPartType _returnvalue_4 "Body" :!class /Script/FortniteGame.FortPlayerPawn))
  (bind _returnvalue_6 (CustomCharacter|GetAssociatedPlayerPawn))
  (bind _returnvalue_7 (Customization|GetSkeletalMeshforPartType _returnvalue_6 "Body" :!class /Script/FortniteGame.FortPlayerPawn))
  (bind _returnvalue_8 (CustomCharacter|GetAssociatedPlayerPawn))
  (bind _returnvalue_9 (Customization|GetSkeletalMeshforPartType _returnvalue_8 "MiscOrTail" :!class /Script/FortniteGame.FortPlayerPawn))
  (bind _returnvalue_10 (CustomCharacter|GetAssociatedPlayerPawn))
  ;[0.02129 0.7064 1] Shark Transformation - Swimming
  (Utilities|FlowControl|Sequence
    (:then_0
      _)
    (:then_1
      (Utilities|FlowControl|DoOnce false
        (:Completed
          (if (CustomCosmeticModifier|IsinFrontEnd)
            (Variables|Default|SetisSwimming true)
            (bind _returnvalue_11 (PartModifier|SpawnSystem|SpawnSystemAttachedPartModifier "/Game/Characters/Player/Female/Medium/Bodies/F_MED_Shapeshifter/FX/NS_Shapeshifter_Charm_ChangeFX.NS_Shapeshifter_Charm_ChangeFX" _returnvalue_2 "None" (Math|Transform|InverseTransformLocation (Transformation|GetWorldTransform _returnvalue_2) _returnvalue_3) "0, 0, 0" "KeepRelativeOffset" true true true "None" false)
              (:Success
                (Niagara|SetNiagaraVariable(Vector3) _returnvalue_11 "WaterPlaneLocation" _returnvalue_3)
                (Niagara|SetNiagaraVariable(Vector3) _returnvalue_11 "WaterPlaneNormal" "0, 0,1.000000")
                (Niagara|SetNiagaraVariable(Float) _returnvalue_11 "WaterBodyIndex" 1.0)
                (Niagara|SetNiagaraVariable(Float) _returnvalue_11 "WaterDepth" 2e+01)
                (Niagara|SetNiagaraSkeletalMeshComponent _returnvalue_11 "SkeletalMesh" (Customization|GetSkeletalMeshforPartType (CustomCharacter|GetAssociatedPlayerPawn) "MiscOrTail" :!class /Script/FortniteGame.FortPlayerPawn))
                (Components|Activation|Activate _returnvalue_11 false))
              (:Failure
                _))
            (Timeline|Timeline_0 PlayFromStart)
            (else
              (Variables|Default|SetisSwimming true)
              (bind _outwatersurfaceinfo (WaterBody|GetWaterSurfaceInfoStruct (WaterBody|GetCurrentWaterBody _returnvalue_10) (Transformation|GetActorLocation _returnvalue_10) true false false))
              (bind (_outwatersurfaceinfo_waterplanelocation _outwatersurfaceinfo_waterplanenormal _outwatersurfaceinfo_watersurfacelocation _outwatersurfaceinfo_waterdepth _outwatersurfaceinfo_waterbodyidx) (Utilities|Struct|BreakWaterSurfaceInfo _outwatersurfaceinfo))
              (bind _returnvalue_12 (PartModifier|SpawnSystem|SpawnSystemAttachedPartModifier "/Game/Characters/Player/Female/Medium/Bodies/F_MED_Shapeshifter/FX/NS_Shapeshifter_Charm_ChangeFX.NS_Shapeshifter_Charm_ChangeFX" (Customization|GetSkeletalMeshforPartType _returnvalue_10 "MiscOrTail" :!class /Script/FortniteGame.FortPlayerPawn) "None" (Math|Transform|InverseTransformLocation (Transformation|GetWorldTransform (Customization|GetSkeletalMeshforPartType _returnvalue_10 "MiscOrTail" :!class /Script/FortniteGame.FortPlayerPawn)) _outwatersurfaceinfo_waterplanelocation) "0, 0, 0" "KeepRelativeOffset" true true true "None" false)
                (:Success
                  (Niagara|SetNiagaraVariable(Vector3) _returnvalue_12 "WaterPlaneLocation" _outwatersurfaceinfo_waterplanelocation)
                  (Niagara|SetNiagaraVariable(Vector3) _returnvalue_12 "WaterPlaneNormal" _outwatersurfaceinfo_waterplanenormal)
                  (Niagara|SetNiagaraVariable(Float) _returnvalue_12 "WaterBodyIndex" (Math|Conversions|ToFloat(Integer) _outwatersurfaceinfo_waterbodyidx))
                  (Niagara|SetNiagaraVariable(Float) _returnvalue_12 "WaterDepth" _outwatersurfaceinfo_waterdepth)
                  (Niagara|SetNiagaraSkeletalMeshComponent _returnvalue_12 "SkeletalMesh" (Customization|GetSkeletalMeshforPartType (CustomCharacter|GetAssociatedPlayerPawn) "MiscOrTail" :!class /Script/FortniteGame.FortPlayerPawn))
                  (Components|Activation|Activate _returnvalue_12 false))
                (:Failure
                  _))))))))
  (Utilities|FlowControl|DoOnce true
    (:Completed
      (Variables|Default|SetisSwimming false)
      (PartModifier|SpawnSystem|SpawnSystemAttachedPartModifier "/Game/Characters/Player/Female/Medium/Bodies/F_MED_Shapeshifter/FX/NS_Shapeshifter_Charm_RevertFX.NS_Shapeshifter_Charm_RevertFX" (Customization|GetSkeletalMeshforPartType (CustomCharacter|GetAssociatedPlayerPawn) "Body" :!class /Script/FortniteGame.FortPlayerPawn) "pelvis" "0, 0, 0" "0, 0, 0" "KeepRelativeOffset" true true true "None" false
        (:Success
          _)
        (:Failure
          _))
      (Rendering|Material|SetVectorParameterValueonMaterials _returnvalue_7 "Gem Slight Glow Color" (Math|Conversions|ToVector(LinearColor) (Variables|Default|GetDefault%20Tattoo)))
      (Rendering|Material|SetScalarParameterValueonMaterials _returnvalue_7 "GemColorTransition" 0.0))))

(event Custom|Snakes
  (params (bIsTargeting :type Boolean))
  (bind _returnvalue (CustomCharacter|GetAssociatedPlayerPawn))
  (bind _returnvalue_2 (Customization|GetSkeletalMeshforPartType _returnvalue "Body" :!class /Script/FortniteGame.FortPlayerPawn))
  (bind _returnvalue_3 (CustomCharacter|GetAssociatedPlayerPawn))
  (bind _returnvalue_4 (Customization|GetSkeletalMeshforPartType _returnvalue_3 "Body" :!class /Script/FortniteGame.FortPlayerPawn))
  (bind (_gem_color _transition) (CallFunction|SettingGemColorTransition))
  ;[1 0.9333 0 0.55] 'Dusa Transformation - ADS
  (if bIsTargeting
    (Utilities|FlowControl|DoOnce false
      (:Completed
        (bind _changefx (Variables|Default|GetChangeFX)
          (:then
            (if (Components|Activation|IsActive _changefx)
              (Components|Activation|Deactivate (Variables|Default|GetChangeFX))
              (Rendering|Material|SetVectorParameterValueonMaterials _returnvalue_4 "Gem Slight Glow Color" (Math|Conversions|ToVector(LinearColor) (Variables|Default|GetADS%20Tattoo)))
              (Rendering|Material|SetScalarParameterValueonMaterials _returnvalue_4 "GemColorTransition" 0.0)
              (bind _returnvalue_5 (PartModifier|SpawnSystem|SpawnSystemAttachedPartModifier "/Game/Characters/Player/Female/Medium/Bodies/F_MED_Shapeshifter/FX/NS_Shapeshifter_Face_ChangeFX.NS_Shapeshifter_Face_ChangeFX" (Customization|GetSkeletalMeshforPartType (CustomCharacter|GetAssociatedPlayerPawn) "Face" :!class /Script/FortniteGame.FortPlayerPawn) "None" "0, 0, 0" "0, 0, 0" "SnapToTarget" true true true "None" false)
                (:Success
                  (bind _output_get (Variables|Default|SetChangeFX _returnvalue_5))
                  (Effects|Components|ParticleSystem|SetBoolParameter _output_get "bisHeadHidden" (Variables|HiddenPartCheck|GetbIsHeadHidden)))
                (:Failure
                  _))
              (Utilities|FlowControl|DoOnce false
                (:Completed
                  (bind _changefx_2 (Variables|Default|GetChangeFX)
                    (:then
                      (Components|Activation|Deactivate _changefx_2)
                      (Variables|Default|SetChangeFX 0)
                      (Rendering|Material|SetVectorParameterValueonMaterials _returnvalue_2 "Gem Slight Glow Color" _gem_color)
                      (Rendering|Material|SetScalarParameterValueonMaterials _returnvalue_2 "GemColorTransition" _transition))
                    (:else
                      _))))))
          (:else
            _))))))

(fn IsSkydivingfromPawn
  (return (Variables|/Script/FortniteGame.FortPlayerPawn|GetbIsSkydiving (Utilities|Casting|CastToFortPlayerPawnAthena (CustomCharacter|GetAssociatedPlayerPawn)))))

(fn SettingGemColorTransition
  (return (select (CallFunction|IsSkydivingFromPawn) (Math|Conversions|ToVector(LinearColor) (Variables|Default|GetSkydiving%20Tattoo)) (select (Variables|Default|GetisSwimming) (Math|Conversions|ToVector(LinearColor) (Variables|Default|GetSwimming%20Tattoo)) (Math|Conversions|ToVector(LinearColor) (Variables|Default|GetDefault%20Tattoo)))) (select (CallFunction|IsSkydivingFromPawn) 0.5 (select (Variables|Default|GetisSwimming) 0.5 0.0))))

(fn ConstructionScript
  (|CallParentFunction))
