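; Part modifier Blueprint for the SilverBellWind cosmetic, applied to head and body character parts.
; After the parent's post-res-in event, it sets the "BodyPart" float parameter on the idle VFX component to 1.
; When in the front end, it also sets an "IsFrontEnd" scalar to 1 on the overlay material of the part's skeletal mesh.
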(event EventOnPostResIn
  (|CallParentFunction)
  (Effects|Components|ParticleSystem|SetFloatParameter (CallFunction|GetIdleVFXComponent) "BodyPart" 1.0)
  (if (CustomCosmeticModifier|IsinFrontEnd)
    (Class|FortCosmeticMaterialBlueprintLibrary|SetScalarParameterValueonOverlayMaterial (CustomCharacter|GetSkeletalMeshForPartOnPlayerPawn) "IsFrontEnd" 1.0)))

(fn ConstructionScript
  (|CallParentFunction))
