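; Cosmetic part modifier for Inkville toon-shaded character parts that applies an inverse front-end LUT to their materials.
; On post res-in, after the parent logic and only in the front end, it creates dynamic instances of the part mesh's materials.
; Where PBR_Shading is not 1, it sets the LUT volume texture, multiplier and related scalars.
; A note marks this as temporary until an auto LUT tool exists.
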(event EventOnPostResIn
  (bind _returnvalue (CustomCharacter|GetSkeletalMeshForPartOnPlayerPawn))
  (|CallParentFunction)
  ;[1 1 1] Front End LUT, Remove after creating auto LUT tool
  (if (CustomCosmeticModifier|IsinFrontEnd)
    (for (_array_element _array_index) (Rendering|Material|GetMaterials _returnvalue)
      (bind _returnvalue_2 (Rendering|Material|CreateDynamicMaterialInstance _returnvalue _array_index _array_element "None" :!class /Script/Engine.PrimitiveComponent))
      (bind _returnvalue_3 (Rendering|Material|GetScalarParameterValue _returnvalue_2 "PBR_Shading" :!class /Script/Engine.MaterialInstanceDynamic))
      (if (not (== _returnvalue_3 1.0))
        (Rendering|Material|SetTextureParameterValue _returnvalue_2 "LUT Volume Texture" "/Game/Packages/Fortress_SharedMaterials/Anime_Master_Material/VT_InverseLUT_FE.VT_InverseLUT_FE")
        (Rendering|Material|SetScalarParameterValue _returnvalue_2 "LUT Multiplier" (Variables|Default|GetLUT%20Multiplier) :!class /Script/Engine.MaterialInstanceDynamic)
        (Rendering|Material|SetScalarParameterValue _returnvalue_2 "Mobile LUT Brightness" 1.0 :!class /Script/Engine.MaterialInstanceDynamic)
        (Rendering|Material|SetScalarParameterValue _returnvalue_2 "use 33 Resolution" 0.0 :!class /Script/Engine.MaterialInstanceDynamic)
        (Rendering|Material|SetScalarParameterValue _returnvalue_2 "use Color Adjustment Instead of LUT" 0.0 :!class /Script/Engine.MaterialInstanceDynamic)))))

(fn ConstructionScript
  (|CallParentFunction))
