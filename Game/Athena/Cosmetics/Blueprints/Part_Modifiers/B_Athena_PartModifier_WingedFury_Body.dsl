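; Character part modifier Blueprint for the Winged Fury body cosmetic that updates its idle VFX when the held weapon changes.
; The weapon-changed event calls the parent handler, then runs the HiddenPartCheck-Weapon helper on the idle VFX component.
; Separate checks use the hands and feet HideFX gameplay tags, each paired with a named hidden flag (bisHandHidden, bisFootHidden).
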(event EventOnWeaponChanged
  (params (NewWeapon :type Fort%20Weapon%20Object%20Reference) (OldWeapon :type Fort%20Weapon%20Object%20Reference))
  (bind _returnvalue (CallFunction|GetIdleVFXComponent))
  (bind _array (Utilities|Array|MakeArray _returnvalue))
  (|CallParentFunction NewWeapon OldWeapon)
  (Default|HiddenPartCheck-Weapon _array NewWeapon OldWeapon true "(TagName=\"Animation.CharacterParts.HideFX.Hands\")" "bisHandHidden")
  (Default|HiddenPartCheck-Weapon _array NewWeapon OldWeapon true "(TagName=\"Animation.CharacterParts.HideFX.Feet\")" "bisFootHidden"))

(fn ConstructionScript)
